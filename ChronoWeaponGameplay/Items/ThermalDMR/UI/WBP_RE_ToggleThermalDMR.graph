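# Reticle extension widget for the Thermal DMR that shows a "Toggle Thermal Vision" keybind prompt while aiming down sights.
# OnTargetingChanged drives UpdateTogglePromptVisibility, which forces the prompt hidden for touch players on mobile.
# Showing plays an opacity intro and starts a looping one-second timer that fades the prompt after SecondsBeforeFade.
# Hiding clears the timer and fades out, jumping to the animation's end if already timed out or unanimated.
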WidgetBlueprint /ChronoWeaponGameplay/Items/ThermalDMR/UI/WBP_RE_ToggleThermalDMR
animations [
 Anim_Intro {
  binding_count=1
  length_seconds=0.3334
  object=&self:Anim_Intro
  sequence {
   bindings [Overlay_Root {
     id=#1
     possessed_class=Overlay
     tracks [{
       type=MovieSceneFloatTrack
       display_name=Render Opacity
       sections [{type=MovieSceneFloatSection, channels [None {type=MovieSceneScriptingFloatChannel, keys [[0,0],[20,1]]}]}]
      }]
    }]
   display_rate {numerator=60}
  }
 }
 Anim_Fade {
  binding_count=1
  length_seconds=0.3334
  object=&self:Anim_Fade
  sequence {
   bindings [Overlay_Root {
     id=#1
     possessed_class=Overlay
     tracks [{
       type=MovieSceneFloatTrack
       display_name=Render Opacity
       sections [{type=MovieSceneFloatSection, channels [None {type=MovieSceneScriptingFloatChannel, keys [[0,1],[20,0]]}]}]
      }]
    }]
   display_rate {numerator=60}
  }
 }
]
graphs [
 ClearTimer {nodes [
   ClearTimer() {is=title K2Node_FunctionEntry}
   > UnbindAllFromAnimationFinished!(
    execute
    self:User Widget Object Reference
    Animation:Widget Animation Object Reference=Get Anim_Fade() -> (Anim_Fade:Widget Animation Object Reference)
   ) -> (then)
   > Clear and Invalidate Timer by Handle(Handle=Get ScopedTimer() -> (ScopedTimer:Timer Handle Structure))
   > Set ScopeTime(ScopeTime:double=0) -> (Output_Get:double)
  ]}
 EventGraph {nodes [
   n1 = Event OnTargetingChanged!() -> (OutputDelegate, then, bIsADS:bool) {is=title K2Node_Event}
   > Set bIsADS(bIsADS:bool=@n1.bIsADS) -> (Output_Get:bool)
   > n2 = Cast To FortPlayerController(Object=GetOwningPlayer())
   > n3 = Branch(Condition=AND Boolean(A=IsUsingTouch(PlayerController=@n2), B=IsMobileGame()))
   > UpdateTogglePromptVisibility!(execute, self, DisplayPrompt:bool=false, Play Animation:bool=false) -> (then)
   UpdateTogglePromptVisibility(execute=[@n2.CastFailed, @n3.else], DisplayPrompt=Get bIsADS() -> (bIsADS:bool), Play Animation=true)
   Event Construct()
   > Set IsFadedFromTimer(IsFadedFromTimer:bool=false) -> (Output_Get:bool)
   > UpdateTogglePromptVisibility(DisplayPrompt=false, Play Animation=false)
   Event Destruct()
   > UpdateTogglePromptVisibility(DisplayPrompt=false, Play Animation=false)
  ]}
 TimerToFade {nodes [
   TimerToFade() {is=title K2Node_FunctionEntry}
   > n1 = Set ScopeTime(ScopeTime=float + float(A=Get ScopeTime() -> (ScopeTime:double), B=1))
   > Branch(Condition="Equal (Float)"(A=@n1, B=Get SecondsBeforeFade() -> (SecondsBeforeFade:double)))
   > BindToAnimationFinished(Animation=Get Anim_Fade(), Delegate=Create Event())
   > PlayAnimation(InAnimation=Get Anim_Fade())
   > Set IsFadedFromTimer(IsFadedFromTimer=true)
  ]}
 UpdateTogglePromptVisibility {nodes [
   n1 = UpdateTogglePromptVisibility!() -> (then, DisplayPrompt:bool=~, Play Animation:bool=~) {is=title K2Node_FunctionEntry}
   > n2 = Branch(Condition=@n1)
   n3 = Get Anim_Intro() -> (Anim_Intro:Widget Animation Object Reference)
   > PlayAnimation(InAnimation=@n3, StartAtTime=SelectFloat(A=0, B=GetEndTime(self=@n3), bPickA=Get Play Animation() -> (Play Animation:bool)))
   > n4 = Set Timer by Event(Delegate=Create Event(), Time=1, bLooping=true)
   > Set ScopedTimer(ScopedTimer:Timer Handle Structure=@n4) -> (Output_Get:Timer Handle Structure)
   ClearTimer(execute=@n2.else) {is=title K2Node_CallFunction}
   n5 = Get Anim_Fade()
   n6 = NOT Boolean(A=Get IsFadedFromTimer() -> (IsFadedFromTimer:bool))
   n7 = Get Play Animation()
   > PlayAnimation(InAnimation=@n5, StartAtTime=SelectFloat(A=0, B=GetEndTime(self=@n5), bPickA=AND Boolean(A=@n7, B=@n6)))
   > Set IsFadedFromTimer(IsFadedFromTimer=false)
   Get SecondsBeforeFade()
  ]}
]
variables [
 bIsADS {type=boolean}
 IsFadedFromTimer {type=boolean}
 ScopedTimer {type=TimerHandle}
 ScopeTime {type=number}
 SecondsBeforeFade {type=number, default=5}
]
widgets {
 info {
  inheritedWidgetCount=1
  parentClass=/Script/FortniteUI.FortWeaponReticleExtensionWidgetBase
  rootWidgetClass=&Overlay
  widgetCount=8
 }
 widgets [
  Overlay_Root {
   widgetClassPath=&Overlay
   widget=Overlay_Root
   properties {visibility=Visible, renderTransform {angle=0, scale=<1 1>, shear=<0 0>, translation=<150 250>}, renderTransformPivot=<0 0>}
  }
  HorizontalBox_0 {
   widgetClassPath=&HorizontalBox
   parent=Overlay_Root
   widget=HorizontalBox_0
   slot {class=OverlaySlot, properties {horizontalAlignment=HAlign_Fill, verticalAlignment=VAlign_Fill}}
  }
  Spacer_109 {widgetClassPath=&Spacer, parent=HorizontalBox_0, widget=Spacer_109, slot {class=HorizontalBoxSlot}}
  SizeBox_0 {
   widgetClassPath=&SizeBox
   parent=HorizontalBox_0
   widget=SizeBox_0
   properties {bOverride_HeightOverride=true, heightOverride=32}
   slot {class=HorizontalBoxSlot}
  }
  Keybind_Toggle {
   widgetClassPath=/FortUILibrary/Input/KeybindWidget
   bIsVariable=true
   parent=SizeBox_0
   widget=Keybind_Toggle
   properties {boundInputAction=/ToggleScopeModeInput/Input/IA_ToggleScopeMode, forcedHoldKeybindStatus=ForcedHold, boundActionFallbacks [UseOrReload]}
   slot {class=SizeBoxSlot, properties {horizontalAlignment=HAlign_Center, verticalAlignment=VAlign_Center}}
  }
  Spacer_67 {widgetClassPath=&Spacer, parent=HorizontalBox_0, widget=Spacer_67, properties {size=<5 0>}, slot {class=HorizontalBoxSlot}}
  Text_TogglePrompt {
   widgetClassPath=&CommonTextBlock
   bIsVariable=true
   parent=HorizontalBox_0
   widget=Text_TogglePrompt
   properties {
    justification=Center
    mobileFontSizeMultiplier=1.2
    text=Toggle Thermal Vision
    visibility=Visible
    wrapTextAt=350
    font {
     bForceMonospaced=false
     bMaterialIsStencil=false
     fontMaterial=null
     fontObject=&BurbankSmall
     letterSpacing=0
     monospacedWidth=1
     size=12
     skewAmount=0
     typefaceFontName=Bold
     outlineSettings {
      bApplyOutlineToDropShadows=true
      bMiteredCorners=false
      bSeparateFillAlpha=true
      outlineMaterial=&M-FontShadow
      outlineSize=3
      outlineColor=<rgba 0 0 0 0.5>
     }
    }
    shadowOffset=<0 0>
   }
   slot {class=HorizontalBoxSlot, properties {horizontalAlignment=HAlign_Left, verticalAlignment=VAlign_Center}}
  }
  TutorialHighlight {widgetClassPath=&TutorialHighlight}
 ]
}
